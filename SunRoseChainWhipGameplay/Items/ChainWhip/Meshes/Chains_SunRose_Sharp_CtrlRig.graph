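# Control Rig for the sharp SunRose chain whip mesh, applying small fixed positional corrections to its bones.
# The forward solve runs four Modify Transforms units in sequence at full weight, with no controls, variables or simulation.
# They nudge the roots of both loop chain halves along Z and the tail chain along X; the chain3 entry has a zero offset.
# Child bones include the chain links and skull-and-jaw club; spawn and bend bones are untouched.
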ControlRigBlueprint /SunRoseChainWhipGameplay/Items/ChainWhip/Meshes/Chains_SunRose_Sharp_CtrlRig
graphs [
 RigVMFunctionLibrary {links [], nodes []}
 RigVMModel {nodes [
   RigUnit_BeginExecution = &RigUnit_BeginExecution()
   &RigUnit_ModifyTransforms(
    ItemToModify=((Item=(Type=Bone,Name="loopChain1"),Transform=(Rotation=(X=0,Y=0,Z=0,W=1),Translation=(X=0,Y=0,Z=-1.5),Scale3D=(X=1,Y=1,Z=1))))
    Weight=1
    ExecutePin=@RigUnit_BeginExecution.ExecutePin
   )
   > &RigUnit_ModifyTransforms(
    ItemToModify=((Item=(Type=Bone,Name="loopChain25"),Transform=(Rotation=(X=0,Y=0,Z=0,W=1),Translation=(X=0,Y=0,Z=-1.5),Scale3D=(X=1,Y=1,Z=1))))
    Weight=1
   )
   > &RigUnit_ModifyTransforms(
    ItemToModify=((Item=(Type=Bone,Name="tailChain1"),Transform=(Rotation=(X=0,Y=0,Z=0,W=1),Translation=(X=1,Y=0,Z=0),Scale3D=(X=1,Y=1,Z=1))))
    Weight=1
   )
   > &RigUnit_ModifyTransforms(
    ItemToModify=((Item=(Type=Bone,Name="chain3"),Transform=(Rotation=(X=0,Y=0,Z=0,W=1),Translation=(X=0,Y=0,Z=0),Scale3D=(X=1,Y=1,Z=1))))
    Weight=1
   )
  ]}
]
hierarchy [
 root {loc=<0 0 0>, rot=<0 0 0 1>}
 chainSpwn1 {parent=root, loc=<-0.7524 -0.6051 0.3616>, rot=<0 0 0 1>}
 chainSpwn2 {parent=chainSpwn1, loc=<34.0733 0 0>, rot=<0 0 0 1>}
 chainSpwn3 {parent=chainSpwn2, loc=<34.0733 0 0>, rot=<0 0 0 1>}
 chainSpwn4 {parent=chainSpwn3, loc=<34.0733 0 0>, rot=<0 0 0 1>}
 chainSpwn5 {parent=chainSpwn4, loc=<34.0733 0 0>, rot=<0 0 0 1>}
 chainSpwn6 {parent=chainSpwn5, loc=<34.0733 0 0>, rot=<0 0 0 1>}
 chainSpwn7 {parent=chainSpwn6, loc=<34.0733 0 0>, rot=<0 0 0 1>}
 chainSpwn8 {parent=chainSpwn7, loc=<34.0733 0 0>, rot=<0 0 0 1>}
 chainSpwn9 {parent=chainSpwn8, loc=<34.0733 0 0>, rot=<0 0 0 1>}
 chainSpwn10 {parent=chainSpwn9, loc=<34.0733 0 0>, rot=<0 0 0 1>}
 chainSpwn11 {parent=chainSpwn10, loc=<34.0733 0 0>, rot=<0 0 0 1>}
 chainSpwn12 {parent=chainSpwn11, loc=<34.0733 0 0>, rot=<0 0 0 1>}
 chainSpwn13 {parent=chainSpwn12, loc=<34.0734 0 0>, rot=<0 0 0 1>}
 chainSpwn14 {parent=chainSpwn13, loc=<34.0733 0 0>, rot=<0 0 0 1>}
 chainSpwn15 {parent=chainSpwn14, loc=<34.0733 0 0>, rot=<0 0 0 1>}
 chainSpwn16 {parent=chainSpwn15, loc=<34.0733 0 0>, rot=<0 0 0 1>}
 loopChain1 {parent=root, loc=<-0.7639 0 -0.4632>, rot=<0.0669 -0.966 -0.0086 -0.2496>}
 loopChain2 {parent=loopChain1, loc=<1.9191 0 0>, rot=<-0.0034 0.0327 -0.0257 -0.9991>}
 loopChain3 {parent=loopChain2, loc=<2.7242 0 0>, rot=<-0.0022 0.1915 0.0175 -0.9813>}
 loopChain4 {parent=loopChain3, loc=<5.6459 0 0>, rot=<0.0031 0.1216 -0.025 -0.9923>}
 loopChain5 {parent=loopChain4, loc=<3.5244 0 0>, rot=<0 0.073 0 -0.9973>}
 loopChain6 {parent=loopChain5, loc=<4.4784 0 0>, rot=<-0.0029 0.0652 0.0031 -0.9979>}
 loopChain7 {parent=loopChain6, loc=<3.695 0 0>, rot=<0.0006 0.0488 -0.0117 -0.9987>}
 loopChain8 {parent=loopChain7, loc=<4.5781 0 0>, rot=<0.0013 0.0206 0.0091 -0.9997>}
 loopChain9 {parent=loopChain8, loc=<3.9639 0 0>, rot=<0.0004 0.0668 -0.0019 -0.9978>}
 loopChain10 {parent=loopChain9, loc=<4.7995 0 0>, rot=<0.0009 0.0872 0.0102 -0.9961>}
 loopChain11 {parent=loopChain10, loc=<3.4365 0 0>, rot=<0 0.0827 0 -0.9966>}
 loopChain12 {parent=loopChain11, loc=<4.7634 0 0>, rot=<0.0031 0.1954 0.0172 -0.9806>}
 loopChain13 {parent=loopChain12, loc=<3.3978 0 0>, rot=<0 0.2018 0 -0.9794>}
 loopChain14 {parent=loopChain13, loc=<4.4977 0 0>, rot=<0.005 0.2673 -0.0181 -0.9634>}
 loopBendMain {parent=root, loc=<-0.0718 -0.4159 0.4498>, rot=<0 -0.7071 0 -0.7071>}
 loopBend1 {parent=loopBendMain, loc=<0 0 0>, rot=<0 0 0 1>}
 loopBend2 {parent=loopBend1, loc=<21.6947 0 0>, rot=<0 0 0 1>}
 tailChain1 {parent=root, loc=<3.3127 -0.5408 0.4499>, rot=<0 0 0.0038 -1>}
 tailChain2 {parent=tailChain1, loc=<4.315 0.0036 0>, rot=<0 0 -0.0009 -1>}
 tailChain3 {parent=tailChain2, loc=<3.2289 0.0024 0>, rot=<0 0 -0.0008 -1>}
 tailChain4 {parent=tailChain3, loc=<3.8868 0.0036 0>, rot=<0 0 -0.0009 -1>}
 tailChain5 {parent=tailChain4, loc=<3.0521 0.0018 0>, rot=<0 0 -0.0005 -1>}
 tailChain6 {parent=tailChain5, loc=<4.1452 0.0012 0>, rot=<0 0 -0.0002 -1>}
 tailChain7 {parent=tailChain6, loc=<3.1468 -0.0011 0>, rot=<0 0 0.0004 -1>}
 tailChain8 {parent=tailChain7, loc=<3.7151 -0.0029 0>, rot=<0 0 0.0008 -1>}
 tailChain9 {parent=tailChain8, loc=<3.4267 -0.0028 0>, rot=<0 0 0.0008 -1>}
 tailChain10 {parent=tailChain9, loc=<3.2309 -0.0023 0>, rot=<0 0 0.0007 -1>}
 tailChain11 {parent=tailChain10, loc=<2.1677 -0.0009 0>, rot=<0 0 0.0004 -1>}
 tailChainClub {parent=tailChain11, loc=<0 0 0>, rot=<0 0 0 1>}
 loopChain25 {parent=root, loc=<4.7725 -1.2184 -2.9764>, rot=<0.0468 -0.9064 -0.0678 0.4142>}
 loopChain24 {parent=loopChain25, loc=<-3.4426 0 0>, rot=<0 0 0 -1>}
 loopChain23 {parent=loopChain24, loc=<-4.8181 -0.3468 1.3183>, rot=<-0.0324 -0.134 -0.0306 -0.99>}
 loopChain22 {parent=loopChain23, loc=<-3.5016 0 0.6193>, rot=<0 -0.0874 0 -0.9962>}
 loopChain21 {parent=loopChain22, loc=<-4.5829 -0.2242 0.6673>, rot=<-0.0165 -0.0726 -0.0231 -0.997>}
 loopChain20 {parent=loopChain21, loc=<-3.5344 0.1488 0.3034>, rot=<-0.0071 -0.0426 0.0213 -0.9988>}
 loopChain19 {parent=loopChain20, loc=<-4.9745 -0.2195 0.4216>, rot=<-0.0111 -0.0425 -0.0215 -0.9988>}
 loopChain18 {parent=loopChain19, loc=<-3.413 0.1486 0.2907>, rot=<0.0048 -0.0425 0.0215 -0.9989>}
 loopChain17 {parent=loopChain18, loc=<-5.33 0 0.7918>, rot=<0 -0.0737 0 -0.9973>}
 loopChain16 {parent=loopChain17, loc=<-2.894 -0.0085 0.737>, rot=<0.025 -0.1243 -0.0046 -0.9919>}
 loopChain15 {parent=loopChain16, loc=<-5.0456 -0.1624 0.8034>, rot=<0.0013 -0.0788 -0.016 -0.9968>}
 chain3 {parent=root, loc=<-0.548 -0.5367 0.4499>, rot=<0 0 0 1>}
 chain2 {parent=chain3, loc=<-16.0559 0.0191 0>, rot=<0 0 0 1>}
 chain1 {parent=chain2, loc=<-16.5033 -0.0191 0>, rot=<0 0 0 1>}
 club {parent=chain1, loc=<-0.0343 0.0203 0.0001>, rot=<0 0 0 -1>}
 jaw {parent=club, loc=<-14.449 -0.0865 -0.8091>, rot=<0 -1 0 0>}
 skullTop {parent=club, loc=<-10.7125 -0.1528 2.1914>, rot=<0 -1 0 0>}
 root {parent=root, type=Socket, loc=<0 0 0>, rot=<0 0 0 1>}
]
