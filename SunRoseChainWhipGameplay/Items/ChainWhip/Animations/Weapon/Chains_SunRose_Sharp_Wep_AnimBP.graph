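# Animation Blueprint for the Sun Rose chain whip weapon, posing its attacks and secondary chain motion.
# A state machine picks idle, three combo attacks or two upper attacks, with at least the first two combos blending slot-fed cached poses by attack range.
# The pose feeds Anim Dynamics, a Control Rig, spring bones mostly scaled by pawn speed, and movement-driven bone tilts.
# bIsTurnOffPhysics switches between this procedural result and the raw state pose.
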AnimBlueprint /SunRoseChainWhipGameplay/Items/ChainWhip/Animations/Weapon/Chains_SunRose_Sharp_Wep_AnimBP
target_skeleton=/SunRoseChainWhipGameplay/Items/ChainWhip/Meshes/Chains_SunRose_Sharp_Skeleton
graphs [
 AnimGraph {nodes [
   n1 = Get PawnVelocitySpeedCutoff() -> (PawnVelocitySpeedCutoff:float)
   n2 = Anim Dynamics - Bone: loopChain1(ComponentPose=Local To Component(LocalPose=Use cached pose 'WeaponState'()), Alpha=@n1, ExternalForce="0,0,0") {
    angularDampingOverride=1
    angularSpringConstant=5000
    bAngularSpring=true
    bOverrideAngularDamping=true
    bUseGravityOverride=true
    gravityScale=0.35
    simulationSpace=RootRelative
    alphaBoolBlend {blendOption=Linear}
    boundBone {boneName=loopChain1}
    componentLinearAccScale=<0.75 0.75 0.75>
    componentLinearVelScale=<0.75 0.75 0.75>
    gravityOverride {z=-1500}
    physicsBodyDefinitions [{
      boundBone {boneName=loopChain1}
      boxExtents=<17 17 17>
      constraintSetup {angularLimitsMax=<30 30 30>, angularLimitsMin=<-30 -30 -30>, angularTarget=<1 0 0>}
      localJointOffset=<1 0 0>
     }]
   }
   n3 = Control Rig(Source=Component To Local(ComponentPose=@n2)) {
    alphaBoolBlend {blendOption=Linear}
    controlRigAssetReference {blueprintRigClass=/SunRoseChainWhipGameplay/Items/ChainWhip/Meshes/Chains_SunRose_Sharp_CtrlRig.Chains_SunRose_Sharp_CtrlRig_C}
   }
   n4 = Spring controller - Bone: tailChain1!(ComponentPose:Component Space Pose Link Structure=Local To Component(LocalPose=@n3), Alpha:float=@n1) -> (Pose:Component Space Pose Link Structure) {
    is=title AnimGraphNode_SpringBone
    springStiffness=100
    alphaBoolBlend {blendOption=Linear}
    alphaScaleBiasClamp {scale=0.005}
    springBone {boneName=tailChain1}
   }
   n5 = Spring controller - Bone: tailChain2!(ComponentPose:Component Space Pose Link Structure=@n4, Alpha:float=@n1) -> (Pose:Component Space Pose Link Structure) {
    is=title AnimGraphNode_SpringBone
    springStiffness=100
    alphaBoolBlend {blendOption=Linear}
    alphaScaleBiasClamp {scale=0.005}
    springBone {boneName=tailChain2}
   }
   n6 = Spring controller - Bone: tailChain3!(ComponentPose:Component Space Pose Link Structure=@n5, Alpha:float=1) -> (Pose:Component Space Pose Link Structure) {
    is=title AnimGraphNode_SpringBone
    springStiffness=100
    alphaBoolBlend {blendOption=Linear}
    alphaScaleBiasClamp {scale=0.005}
    springBone {boneName=tailChain3}
   }
   n7 = Spring controller - Bone: tailChain4!(ComponentPose:Component Space Pose Link Structure=@n6, Alpha:float=1) -> (Pose:Component Space Pose Link Structure) {
    is=title AnimGraphNode_SpringBone
    springStiffness=100
    alphaBoolBlend {blendOption=Linear}
    alphaScaleBiasClamp {scale=0.005}
    springBone {boneName=tailChain4}
   }
   n8 = Spring controller - Bone: tailChain5!(ComponentPose:Component Space Pose Link Structure=@n7, Alpha:float=@n1) -> (Pose:Component Space Pose Link Structure) {
    is=title AnimGraphNode_SpringBone
    springStiffness=100
    alphaBoolBlend {blendOption=Linear}
    alphaScaleBiasClamp {scale=0.005}
    springBone {boneName=tailChain5}
   }
   n9 = Spring controller - Bone: tailChain6!(ComponentPose:Component Space Pose Link Structure=@n8, Alpha:float=@n1) -> (Pose:Component Space Pose Link Structure) {
    is=title AnimGraphNode_SpringBone
    springStiffness=100
    alphaBoolBlend {blendOption=Linear}
    alphaScaleBiasClamp {scale=0.005}
    springBone {boneName=tailChain6}
   }
   n10 = Spring controller - Bone: tailChain7!(ComponentPose:Component Space Pose Link Structure=@n9, Alpha:float=@n1) -> (Pose:Component Space Pose Link Structure) {
    is=title AnimGraphNode_SpringBone
    springStiffness=100
    alphaBoolBlend {blendOption=Linear}
    alphaScaleBiasClamp {scale=0.005}
    springBone {boneName=tailChain7}
   }
   n11 = Spring controller - Bone: tailChain8!(ComponentPose:Component Space Pose Link Structure=@n10, Alpha:float=@n1) -> (Pose:Component Space Pose Link Structure) {
    is=title AnimGraphNode_SpringBone
    springStiffness=100
    alphaBoolBlend {blendOption=Linear}
    alphaScaleBiasClamp {scale=0.005}
    springBone {boneName=tailChain8}
   }
   n12 = Spring controller - Bone: tailChain9!(ComponentPose:Component Space Pose Link Structure=@n11, Alpha:float=@n1) -> (Pose:Component Space Pose Link Structure) {
    is=title AnimGraphNode_SpringBone
    springStiffness=100
    alphaBoolBlend {blendOption=Linear}
    alphaScaleBiasClamp {scale=0.005}
    springBone {boneName=tailChain9}
   }
   n13 = Spring controller - Bone: tailChain10!(ComponentPose:Component Space Pose Link Structure=@n12, Alpha:float=@n1) -> (Pose:Component Space Pose Link Structure) {
    is=title AnimGraphNode_SpringBone
    springStiffness=100
    alphaBoolBlend {blendOption=Linear}
    alphaScaleBiasClamp {scale=0.005}
    springBone {boneName=tailChain10}
   }
   n14 = Spring controller - Bone: tailChain11!(ComponentPose:Component Space Pose Link Structure=@n13, Alpha:float=@n1) -> (Pose:Component Space Pose Link Structure) {
    is=title AnimGraphNode_SpringBone
    springStiffness=100
    alphaBoolBlend {blendOption=Linear}
    alphaScaleBiasClamp {scale=0.005}
    springBone {boneName=tailChain11}
   }
   n15 = Spring controller - Bone: chain3!(ComponentPose:Component Space Pose Link Structure=@n14, Alpha:float=@n1) -> (Pose:Component Space Pose Link Structure) {
    is=title AnimGraphNode_SpringBone
    springDamping=10
    springStiffness=200
    alphaBoolBlend {blendOption=Linear}
    alphaScaleBiasClamp {scale=0.03}
    springBone {boneName=chain3}
   }
   n16 = Spring controller - Bone: chain2!(ComponentPose:Component Space Pose Link Structure=@n15, Alpha:float=@n1) -> (Pose:Component Space Pose Link Structure) {
    is=title AnimGraphNode_SpringBone
    springDamping=10
    springStiffness=200
    alphaBoolBlend {blendOption=Linear}
    alphaScaleBiasClamp {scale=0.1}
    springBone {boneName=chain2}
   }
   n17 = Spring controller - Bone: chain1!(ComponentPose:Component Space Pose Link Structure=@n16, Alpha:float=@n1) -> (Pose:Component Space Pose Link Structure) {
    is=title AnimGraphNode_SpringBone
    springDamping=10
    springStiffness=200
    alphaBoolBlend {blendOption=Linear}
    alphaScaleBiasClamp {scale=0.05}
    springBone {boneName=chain1}
   }
   n18 = Spring controller - Bone: club!(ComponentPose:Component Space Pose Link Structure=@n17, Alpha:float=@n1) -> (Pose:Component Space Pose Link Structure) {
    is=title AnimGraphNode_SpringBone
    springDamping=6
    springStiffness=200
    alphaBoolBlend {blendOption=Linear}
    alphaScaleBiasClamp {scale=0.01}
    springBone {boneName=club}
   }
   n19 = Get PawnVelocitySpeedCutoff()
   n20 = Spring controller - Bone: loopChain14!(ComponentPose:Component Space Pose Link Structure=@n18, Alpha:float=@n19) -> (Pose:Component Space Pose Link Structure) {
    is=title AnimGraphNode_SpringBone
    springDamping=10
    springStiffness=300
    alphaBoolBlend {blendOption=Linear}
    alphaScaleBiasClamp {scale=0.01}
    springBone {boneName=loopChain14}
   }
   n21 = Spring controller - Bone: loopChain15!(ComponentPose:Component Space Pose Link Structure=@n20, Alpha:float=@n19) -> (Pose:Component Space Pose Link Structure) {
    is=title AnimGraphNode_SpringBone
    springDamping=10
    springStiffness=300
    alphaBoolBlend {blendOption=Linear}
    alphaScaleBiasClamp {scale=0.01}
    springBone {boneName=loopChain15}
   }
   n22 = Spring controller - Bone: loopChain13!(ComponentPose:Component Space Pose Link Structure=@n21, Alpha:float=@n19) -> (Pose:Component Space Pose Link Structure) {
    is=title AnimGraphNode_SpringBone
    springDamping=10
    springStiffness=300
    alphaBoolBlend {blendOption=Linear}
    alphaScaleBiasClamp {scale=0.01}
    springBone {boneName=loopChain13}
   }
   n23 = Spring controller - Bone: loopChain16!(ComponentPose:Component Space Pose Link Structure=@n22, Alpha:float=@n19) -> (Pose:Component Space Pose Link Structure) {
    is=title AnimGraphNode_SpringBone
    springDamping=10
    springStiffness=300
    alphaBoolBlend {blendOption=Linear}
    alphaScaleBiasClamp {scale=0.01}
    springBone {boneName=loopChain16}
   }
   n24 = Spring controller - Bone: loopChain12!(ComponentPose:Component Space Pose Link Structure=@n23, Alpha:float=@n19) -> (Pose:Component Space Pose Link Structure) {
    is=title AnimGraphNode_SpringBone
    springDamping=10
    springStiffness=300
    alphaBoolBlend {blendOption=Linear}
    alphaScaleBiasClamp {scale=0.01}
    springBone {boneName=loopChain12}
   }
   n25 = Spring controller - Bone: loopChain17!(ComponentPose:Component Space Pose Link Structure=@n24, Alpha:float=@n19) -> (Pose:Component Space Pose Link Structure) {
    is=title AnimGraphNode_SpringBone
    springDamping=10
    springStiffness=300
    alphaBoolBlend {blendOption=Linear}
    alphaScaleBiasClamp {scale=0.01}
    springBone {boneName=loopChain17}
   }
   n26 = Spring controller - Bone: loopChain11!(ComponentPose:Component Space Pose Link Structure=@n25, Alpha:float=@n19) -> (Pose:Component Space Pose Link Structure) {
    is=title AnimGraphNode_SpringBone
    springDamping=10
    springStiffness=300
    alphaBoolBlend {blendOption=Linear}
    alphaScaleBiasClamp {scale=0.01}
    springBone {boneName=loopChain11}
   }
   n27 = Spring controller - Bone: loopChain18!(ComponentPose:Component Space Pose Link Structure=@n26, Alpha:float=@n19) -> (Pose:Component Space Pose Link Structure) {
    is=title AnimGraphNode_SpringBone
    springDamping=10
    springStiffness=300
    alphaBoolBlend {blendOption=Linear}
    alphaScaleBiasClamp {scale=0.01}
    springBone {boneName=loopChain18}
   }
   n28 = Spring controller - Bone: loopChain10!(ComponentPose:Component Space Pose Link Structure=@n27, Alpha:float=@n19) -> (Pose:Component Space Pose Link Structure) {
    is=title AnimGraphNode_SpringBone
    springDamping=10
    springStiffness=300
    alphaBoolBlend {blendOption=Linear}
    alphaScaleBiasClamp {scale=0.01}
    springBone {boneName=loopChain10}
   }
   n29 = Spring controller - Bone: loopChain19!(ComponentPose:Component Space Pose Link Structure=@n28, Alpha:float=@n19) -> (Pose:Component Space Pose Link Structure) {
    is=title AnimGraphNode_SpringBone
    springDamping=10
    springStiffness=300
    alphaBoolBlend {blendOption=Linear}
    alphaScaleBiasClamp {scale=0.01}
    springBone {boneName=loopChain19}
   }
   n30 = Spring controller - Bone: loopChain9!(ComponentPose:Component Space Pose Link Structure=@n29, Alpha:float=@n19) -> (Pose:Component Space Pose Link Structure) {
    is=title AnimGraphNode_SpringBone
    springDamping=10
    springStiffness=300
    alphaBoolBlend {blendOption=Linear}
    alphaScaleBiasClamp {scale=0.01}
    springBone {boneName=loopChain9}
   }
   n31 = Spring controller - Bone: loopChain20!(ComponentPose:Component Space Pose Link Structure=@n30, Alpha:float=@n19) -> (Pose:Component Space Pose Link Structure) {
    is=title AnimGraphNode_SpringBone
    springDamping=10
    springStiffness=300
    alphaBoolBlend {blendOption=Linear}
    alphaScaleBiasClamp {scale=0.01}
    springBone {boneName=loopChain20}
   }
   n32 = Spring controller - Bone: loopChain8!(ComponentPose:Component Space Pose Link Structure=@n31, Alpha:float=@n19) -> (Pose:Component Space Pose Link Structure) {
    is=title AnimGraphNode_SpringBone
    springDamping=10
    springStiffness=300
    alphaBoolBlend {blendOption=Linear}
    alphaScaleBiasClamp {scale=0.01}
    springBone {boneName=loopChain8}
   }
   n33 = Spring controller - Bone: loopChain21!(ComponentPose:Component Space Pose Link Structure=@n32, Alpha:float=@n19) -> (Pose:Component Space Pose Link Structure) {
    is=title AnimGraphNode_SpringBone
    springDamping=10
    springStiffness=300
    alphaBoolBlend {blendOption=Linear}
    alphaScaleBiasClamp {scale=0.01}
    springBone {boneName=loopChain21}
   }
   n34 = Spring controller - Bone: loopChain7!(ComponentPose:Component Space Pose Link Structure=@n33, Alpha:float=@n19) -> (Pose:Component Space Pose Link Structure) {
    is=title AnimGraphNode_SpringBone
    springDamping=10
    springStiffness=300
    alphaBoolBlend {blendOption=Linear}
    alphaScaleBiasClamp {scale=0.01}
    springBone {boneName=loopChain7}
   }
   n35 = Spring controller - Bone: loopChain22!(ComponentPose:Component Space Pose Link Structure=@n34, Alpha:float=@n19) -> (Pose:Component Space Pose Link Structure) {
    is=title AnimGraphNode_SpringBone
    springDamping=10
    springStiffness=300
    alphaBoolBlend {blendOption=Linear}
    alphaScaleBiasClamp {scale=0.01}
    springBone {boneName=loopChain22}
   }
   n36 = Spring controller - Bone: loopChain6!(ComponentPose:Component Space Pose Link Structure=@n35, Alpha:float=@n19) -> (Pose:Component Space Pose Link Structure) {
    is=title AnimGraphNode_SpringBone
    springDamping=10
    springStiffness=300
    alphaBoolBlend {blendOption=Linear}
    alphaScaleBiasClamp {scale=0.01}
    springBone {boneName=loopChain6}
   }
   n37 = Spring controller - Bone: loopChain23!(ComponentPose:Component Space Pose Link Structure=@n36, Alpha:float=@n19) -> (Pose:Component Space Pose Link Structure) {
    is=title AnimGraphNode_SpringBone
    springDamping=10
    springStiffness=300
    alphaBoolBlend {blendOption=Linear}
    alphaScaleBiasClamp {scale=0.01}
    springBone {boneName=loopChain23}
   }
   n38 = Spring controller - Bone: loopChain5!(ComponentPose:Component Space Pose Link Structure=@n37, Alpha:float=@n19) -> (Pose:Component Space Pose Link Structure) {
    is=title AnimGraphNode_SpringBone
    springDamping=10
    springStiffness=300
    alphaBoolBlend {blendOption=Linear}
    alphaScaleBiasClamp {scale=0.01}
    springBone {boneName=loopChain5}
   }
   n39 = Spring controller - Bone: loopChain24!(ComponentPose:Component Space Pose Link Structure=@n38, Alpha:float=@n19) -> (Pose:Component Space Pose Link Structure) {
    is=title AnimGraphNode_SpringBone
    springDamping=10
    springStiffness=300
    alphaBoolBlend {blendOption=Linear}
    alphaScaleBiasClamp {scale=0.01}
    springBone {boneName=loopChain24}
   }
   n40 = Spring controller - Bone: loopChain4!(ComponentPose:Component Space Pose Link Structure=@n39, Alpha:float=@n19) -> (Pose:Component Space Pose Link Structure) {
    is=title AnimGraphNode_SpringBone
    springDamping=10
    springStiffness=300
    alphaBoolBlend {blendOption=Linear}
    alphaScaleBiasClamp {scale=0.01}
    springBone {boneName=loopChain4}
   }
   n41 = Spring controller - Bone: loopChain25!(ComponentPose:Component Space Pose Link Structure=@n40, Alpha:float=@n19) -> (Pose:Component Space Pose Link Structure) {
    is=title AnimGraphNode_SpringBone
    springDamping=10
    springStiffness=300
    alphaBoolBlend {blendOption=Linear}
    alphaScaleBiasClamp {scale=0.01}
    springBone {boneName=loopChain25}
   }
   n42 = Spring controller - Bone: loopChain3!(ComponentPose:Component Space Pose Link Structure=@n41, Alpha:float=@n19) -> (Pose:Component Space Pose Link Structure) {
    is=title AnimGraphNode_SpringBone
    springDamping=10
    springStiffness=300
    alphaBoolBlend {blendOption=Linear}
    alphaScaleBiasClamp {scale=0.01}
    springBone {boneName=loopChain3}
   }
   n43 = Spring controller - Bone: loopChain2!(ComponentPose:Component Space Pose Link Structure=@n42, Alpha:float=@n19) -> (Pose:Component Space Pose Link Structure) {
    is=title AnimGraphNode_SpringBone
    springDamping=10
    springStiffness=300
    alphaBoolBlend {blendOption=Linear}
    alphaScaleBiasClamp {scale=0.01}
    springBone {boneName=loopChain2}
   }
   n44 = Anim Dynamics - Bone: tailChain5(ComponentPose=@n43, Alpha=1, ExternalForce="0,0,0") {
    angularDampingOverride=1
    angularSpringConstant=5000
    bAngularSpring=true
    bOverrideAngularDamping=true
    bUseGravityOverride=true
    gravityScale=0.35
    simulationSpace=RootRelative
    alphaBoolBlend {blendOption=Linear}
    boundBone {boneName=tailChain5}
    componentLinearAccScale=<0.4 0.4 0.4>
    componentLinearVelScale=<0.4 0.4 0.4>
    gravityOverride {z=-1500}
    physicsBodyDefinitions [{
      boundBone {boneName=tailChain5}
      boxExtents=<25 25 25>
      constraintSetup {angularLimitsMax=<30 30 30>, angularLimitsMin=<-30 -30 -30>, angularTarget=<1 0 0>}
      localJointOffset=<5 0 0>
     }]
   }
   n45 = Anim Dynamics - Bone: tailChain6(ComponentPose=@n44, Alpha=1, ExternalForce="0,0,0") {
    angularDampingOverride=1
    angularSpringConstant=5000
    bAngularSpring=true
    bOverrideAngularDamping=true
    bUseGravityOverride=true
    gravityScale=0.35
    simulationSpace=RootRelative
    alphaBoolBlend {blendOption=Linear}
    boundBone {boneName=tailChain6}
    componentLinearAccScale=<0.4 0.4 0.4>
    componentLinearVelScale=<0.4 0.4 0.4>
    gravityOverride {z=-1500}
    physicsBodyDefinitions [{
      boundBone {boneName=tailChain6}
      boxExtents=<25 25 25>
      constraintSetup {angularLimitsMax=<30 30 30>, angularLimitsMin=<-30 -30 -30>, angularTarget=<1 0 0>}
      localJointOffset=<5 0 0>
     }]
   }
   n46 = Anim Dynamics - Bone: tailChain7(ComponentPose=@n45, Alpha=1, ExternalForce="0,0,0") {
    angularDampingOverride=1
    angularSpringConstant=5000
    bAngularSpring=true
    bOverrideAngularDamping=true
    bUseGravityOverride=true
    gravityScale=0.35
    simulationSpace=RootRelative
    alphaBoolBlend {blendOption=Linear}
    boundBone {boneName=tailChain7}
    componentLinearAccScale=<0.4 0.4 0.4>
    componentLinearVelScale=<0.4 0.4 0.4>
    gravityOverride {z=-1500}
    physicsBodyDefinitions [{
      boundBone {boneName=tailChain7}
      boxExtents=<25 25 25>
      constraintSetup {angularLimitsMax=<30 30 30>, angularLimitsMin=<-30 -30 -30>, angularTarget=<1 0 0>}
      localJointOffset=<5 0 0>
     }]
   }
   n47 = Anim Dynamics - Bone: tailChain8(ComponentPose=@n46, Alpha=1, ExternalForce="0,0,0") {
    angularDampingOverride=1
    angularSpringConstant=5000
    bAngularSpring=true
    bOverrideAngularDamping=true
    bUseGravityOverride=true
    gravityScale=0.35
    simulationSpace=RootRelative
    alphaBoolBlend {blendOption=Linear}
    boundBone {boneName=tailChain8}
    componentLinearAccScale=<0.4 0.4 0.4>
    componentLinearVelScale=<0.4 0.4 0.4>
    gravityOverride {z=-1500}
    physicsBodyDefinitions [{
      boundBone {boneName=tailChain8}
      boxExtents=<25 25 25>
      constraintSetup {angularLimitsMax=<30 30 30>, angularLimitsMin=<-30 -30 -30>, angularTarget=<1 0 0>}
      localJointOffset=<5 0 0>
     }]
   }
   n48 = Anim Dynamics - Bone: tailChain9(ComponentPose=@n47, Alpha=1, ExternalForce="0,0,0") {
    angularDampingOverride=1
    angularSpringConstant=5000
    bAngularSpring=true
    bOverrideAngularDamping=true
    bUseGravityOverride=true
    gravityScale=0.35
    simulationSpace=RootRelative
    alphaBoolBlend {blendOption=Linear}
    boundBone {boneName=tailChain9}
    componentLinearAccScale=<0.4 0.4 0.4>
    componentLinearVelScale=<0.4 0.4 0.4>
    gravityOverride {z=-1500}
    physicsBodyDefinitions [{
      boundBone {boneName=tailChain9}
      boxExtents=<25 25 25>
      constraintSetup {angularLimitsMax=<30 30 30>, angularLimitsMin=<-30 -30 -30>, angularTarget=<1 0 0>}
      localJointOffset=<5 0 0>
     }]
   }
   n49 = Anim Dynamics - Bone: tailChain10(ComponentPose=@n48, Alpha=1, ExternalForce="0,0,0") {
    angularDampingOverride=1
    angularSpringConstant=5000
    bAngularSpring=true
    bOverrideAngularDamping=true
    bUseGravityOverride=true
    gravityScale=0.35
    simulationSpace=RootRelative
    alphaBoolBlend {blendOption=Linear}
    boundBone {boneName=tailChain10}
    componentLinearAccScale=<0.4 0.4 0.4>
    componentLinearVelScale=<0.4 0.4 0.4>
    gravityOverride {z=-1500}
    physicsBodyDefinitions [{
      boundBone {boneName=tailChain10}
      boxExtents=<25 25 25>
      constraintSetup {angularLimitsMax=<30 30 30>, angularLimitsMin=<-30 -30 -30>, angularTarget=<1 0 0>}
      localJointOffset=<5 0 0>
     }]
   }
   n50 = Anim Dynamics - Bone: tailChain11(ComponentPose=@n49, Alpha=1, ExternalForce="0,0,0") {
    angularDampingOverride=1
    angularSpringConstant=5000
    bAngularSpring=true
    bOverrideAngularDamping=true
    bUseGravityOverride=true
    gravityScale=0.35
    simulationSpace=RootRelative
    alphaBoolBlend {blendOption=Linear}
    boundBone {boneName=tailChain11}
    componentLinearAccScale=<0.4 0.4 0.4>
    componentLinearVelScale=<0.4 0.4 0.4>
    gravityOverride {z=-1500}
    physicsBodyDefinitions [{
      boundBone {boneName=tailChain11}
      boxExtents=<25 25 25>
      constraintSetup {angularLimitsMax=<30 30 30>, angularLimitsMin=<-30 -30 -30>, angularTarget=<1 0 0>}
      localJointOffset=<5 0 0>
     }]
   }
   n51 = Anim Dynamics - Bone: tailChainClub(ComponentPose=@n50, Alpha=1, ExternalForce="0,0,0") {
    angularDampingOverride=1
    angularSpringConstant=5000
    bAngularSpring=true
    bOverrideAngularDamping=true
    bUseGravityOverride=true
    gravityScale=0.35
    simulationSpace=RootRelative
    alphaBoolBlend {blendOption=Linear}
    boundBone {boneName=tailChainClub}
    componentLinearAccScale=<0.4 0.4 0.4>
    componentLinearVelScale=<0.4 0.4 0.4>
    gravityOverride {z=-1500}
    physicsBodyDefinitions [{
      boundBone {boneName=tailChainClub}
      boxExtents=<15 15 15>
      constraintSetup {angularLimitsMax=<30 30 30>, angularLimitsMin=<-30 -30 -30>, angularTarget=<1 0 0>}
      localJointOffset=<5 0 0>
     }]
   }
   n52 = Anim Dynamics - Bone: chain2(ComponentPose=@n51, Alpha=1, ExternalForce="0,0,0") {
    angularDampingOverride=1
    angularSpringConstant=5000
    bAngularSpring=true
    bOverrideAngularDamping=true
    bUseGravityOverride=true
    gravityScale=0.35
    simulationSpace=RootRelative
    alphaBoolBlend {blendOption=Linear}
    boundBone {boneName=chain2}
    componentLinearAccScale=<0.75 0.75 0.75>
    componentLinearVelScale=<0.75 0.75 0.75>
    gravityOverride {z=-1500}
    physicsBodyDefinitions [{
      boundBone {boneName=chain2}
      boxExtents=<25 25 25>
      constraintSetup {angularLimitsMax=<60 60 60>, angularLimitsMin=<-60 -60 -60>, angularTarget=<1 0 0>}
      localJointOffset=<-5 0 0>
     }]
   }
   n53 = Anim Dynamics - Bone: chain1(ComponentPose=@n52, Alpha=1, ExternalForce="0,0,0") {
    angularDampingOverride=1
    angularSpringConstant=5000
    bAngularSpring=true
    bOverrideAngularDamping=true
    bUseGravityOverride=true
    gravityScale=0.35
    simulationSpace=RootRelative
    alphaBoolBlend {blendOption=Linear}
    boundBone {boneName=chain1}
    componentLinearAccScale=<0.75 0.75 0.75>
    componentLinearVelScale=<0.75 0.75 0.75>
    gravityOverride {z=-1500}
    physicsBodyDefinitions [{
      boundBone {boneName=chain1}
      boxExtents=<25 25 25>
      constraintSetup {angularLimitsMax=<60 60 60>, angularLimitsMin=<-60 -60 -60>, angularTarget=<1 0 0>}
      localJointOffset=<-5 0 0>
     }]
   }
   n54 = Get XMoveLeft() -> (XMoveLeft:float)
   n55 = "Transform (Modify) Bone - Bone: chain2"(ComponentPose=@n53, Alpha=@n54, Translation="0,0,0", Rotation="-10,0,0", Scale="1,1,1") {
    rotationMode=BMM_Additive
    alphaBoolBlend {blendOption=Linear}
    alphaScaleBiasClamp {bInterpResult=true, interpSpeedDecreasing=1}
    boneToModify {boneName=chain2}
    rotation {pitch=-10}
   }
   n56 = "Transform (Modify) Bone - Bone: tailChain3"(ComponentPose=@n55, Alpha=@n54, Translation="0,0,0", Rotation="-5,0,0", Scale="1,1,1") {
    rotationMode=BMM_Additive
    alphaBoolBlend {blendOption=Linear}
    alphaScaleBiasClamp {bInterpResult=true, interpSpeedDecreasing=1}
    boneToModify {boneName=tailChain3}
    rotation {pitch=-5}
   }
   n57 = Get XMoveRight() -> (XMoveRight:float)
   n58 = "Transform (Modify) Bone - Bone: chain2"(ComponentPose=@n56, Alpha=@n57, Translation="0,0,0", Rotation="-5,0,0", Scale="1,1,1") {
    rotationMode=BMM_Additive
    alphaBoolBlend {blendOption=Linear}
    alphaScaleBiasClamp {bInterpResult=true}
    boneToModify {boneName=chain2}
    rotation {pitch=-5}
   }
   n59 = "Transform (Modify) Bone - Bone: tailChain3"(ComponentPose=@n58, Alpha=@n57, Translation="0,0,0", Rotation="-5,0,0", Scale="1,1,1") {
    rotationMode=BMM_Additive
    alphaBoolBlend {blendOption=Linear}
    alphaScaleBiasClamp {bInterpResult=true, interpSpeedDecreasing=1}
    boneToModify {boneName=tailChain3}
    rotation {pitch=-5}
   }
   n60 = Get YMoveFwd() -> (YMoveFwd:float)
   n61 = "Transform (Modify) Bone - Bone: chain2"(ComponentPose=@n59, Alpha=@n60, Translation="0,0,0", Rotation="0,30,0", Scale="1,1,1") {
    rotationMode=BMM_Additive
    alphaBoolBlend {blendOption=Linear}
    alphaScaleBiasClamp {bInterpResult=true, interpSpeedDecreasing=1, interpSpeedIncreasing=1}
    boneToModify {boneName=chain2}
    rotation {yaw=30}
   }
   n62 = "Transform (Modify) Bone - Bone: tailChain3"(ComponentPose=@n61, Alpha=@n60, Translation="0,0,0", Rotation="-10,20,0", Scale="1,1,1") {
    rotationMode=BMM_Additive
    alphaBoolBlend {blendOption=Linear}
    alphaScaleBiasClamp {bInterpResult=true, interpSpeedDecreasing=1}
    boneToModify {boneName=tailChain3}
    rotation {pitch=-10, yaw=20}
   }
   Output Pose(Result=Blend Poses by bool(
     BlendPose_0=Use cached pose 'WeaponState'()
     BlendPose_1=Component To Local(ComponentPose=@n62)
     BlendTime_0=0.2
     BlendTime_1=0.35
     bActiveValue=Get bIsTurnOffPhysics() -> (bIsTurnOffPhysics:bool)
    ) {blendType=Linear}) {layerGroup=None}
   Save cached pose '1500'(Pose=Slot '1500'(Source=Sequence Player 'Hades_Weapon_NonTargeted'() {bLoopAnimation=false, sequence=&/Hades_Weapon_NonTargeted}) {slotName=1500}) {cacheName=1500}
   Save cached pose '700'(Pose=Slot '700'(Source=Sequence Player 'Hades_Weapon_NonTargeted'() {bLoopAnimation=false, sequence=&Hades_Weapon_NonTargeted}) {slotName=700}) {cacheName=700}
   Save cached pose '200'(Pose=Slot '200'(Source=Sequence Player 'Hades_Weapon_NonTargeted'() {bLoopAnimation=false, sequence=&Hades_Weapon_NonTargeted}) {slotName=200}) {cacheName=200}
   AnimGraphNode_StateMachine_0 = Chain State Machine() {is=title AnimGraphNode_StateMachine, bAllowConduitEntryStates=true}
   Save cached pose 'WeaponState'(Pose=Slot 'DefaultSlot'(Source=@AnimGraphNode_StateMachine_0)) {cacheName=WeaponState}
   Save cached pose '1200'(Pose=Slot '1200'(Source=Sequence Player 'Hades_Weapon_NonTargeted'() {bLoopAnimation=false, sequence=&Hades_Weapon_NonTargeted}) {slotName=1200}) {cacheName=1200}
  ]}
 Chain State Machine {
  entry=Conduit
  path=AnimGraph.AnimGraphNode_StateMachine_0.Chain State Machine
  ids=id
  nodes [
   AnimStateConduitNode_0 = Conduit()
   AnimStateNode_0 = Idle() {bAlwaysResetOnEntry=true}
   AnimStateNode_1 = Attack_01() {is=title AnimStateNode, bAlwaysResetOnEntry=true}
   AnimStateNode_2 = Attack_02() {is=title AnimStateNode, bAlwaysResetOnEntry=true}
   AnimStateNode_3 = Attack_03() {is=title AnimStateNode, bAlwaysResetOnEntry=true}
   AnimStateNode_4 = Attack_01_Upper() {is=title AnimStateNode}
   AnimStateNode_5 = Attack_02_Upper() {is=title AnimStateNode}
   AnimStateTransitionNode_0 = AnimStateTransitionNode() {crossfadeDuration=0.25, priorityOrder=100}
   AnimStateTransitionNode_1 = AnimStateTransitionNode()
   AnimStateTransitionNode_10 = AnimStateTransitionNode() {automaticRuleTriggerTime=0, bAutomaticRuleBasedOnSequencePlayerInState=true}
   AnimStateTransitionNode_11 = AnimStateTransitionNode() {automaticRuleTriggerTime=0, bAutomaticRuleBasedOnSequencePlayerInState=true}
   AnimStateTransitionNode_12 = AnimStateTransitionNode() {crossfadeDuration=0, priorityOrder=2}
   AnimStateTransitionNode_13 = AnimStateTransitionNode() {crossfadeDuration=0}
   AnimStateTransitionNode_14 = AnimStateTransitionNode() {bAutomaticRuleBasedOnSequencePlayerInState=true}
   AnimStateTransitionNode_15 = AnimStateTransitionNode() {bAutomaticRuleBasedOnSequencePlayerInState=true}
   AnimStateTransitionNode_16 = AnimStateTransitionNode()
   AnimStateTransitionNode_17 = AnimStateTransitionNode()
   AnimStateTransitionNode_2 = AnimStateTransitionNode() {blendMode=Linear, crossfadeDuration=0}
   AnimStateTransitionNode_3 = AnimStateTransitionNode()
   AnimStateTransitionNode_4 = AnimStateTransitionNode() {blendMode=Linear, crossfadeDuration=0}
   AnimStateTransitionNode_5 = AnimStateTransitionNode() {blendMode=Linear, crossfadeDuration=0}
   AnimStateTransitionNode_6 = AnimStateTransitionNode()
   AnimStateTransitionNode_7 = AnimStateTransitionNode()
   AnimStateTransitionNode_8 = AnimStateTransitionNode() {blendMode=Linear, crossfadeDuration=0}
   AnimStateTransitionNode_9 = AnimStateTransitionNode() {automaticRuleTriggerTime=0, bAutomaticRuleBasedOnSequencePlayerInState=true}
  ]
  links [
   {
    dst=Attack_02
    rule=AnimGraph.AnimGraphNode_StateMachine_0.Chain State Machine.AnimStateTransitionNode_4.Transition
    src=Attack_01
    via=AnimStateTransitionNode_4
   }
   {
    dst=Conduit
    rule=AnimGraph.AnimGraphNode_StateMachine_0.Chain State Machine.AnimStateTransitionNode_11.Transition
    src=Attack_01
    via=AnimStateTransitionNode_11
   }
   {
    dst=Conduit
    rule=AnimGraph.AnimGraphNode_StateMachine_0.Chain State Machine.AnimStateTransitionNode_3.Transition
    src=Attack_01
    via=AnimStateTransitionNode_3
   }
   {
    dst=Attack_02_Upper
    rule=AnimGraph.AnimGraphNode_StateMachine_0.Chain State Machine.AnimStateTransitionNode_13.Transition
    src=Attack_01_Upper
    via=AnimStateTransitionNode_13
   }
   {
    dst=Conduit
    rule=AnimGraph.AnimGraphNode_StateMachine_0.Chain State Machine.AnimStateTransitionNode_15.Transition
    src=Attack_01_Upper
    via=AnimStateTransitionNode_15
   }
   {
    dst=Conduit
    rule=AnimGraph.AnimGraphNode_StateMachine_0.Chain State Machine.AnimStateTransitionNode_17.Transition
    src=Attack_01_Upper
    via=AnimStateTransitionNode_17
   }
   {
    dst=Attack_03
    rule=AnimGraph.AnimGraphNode_StateMachine_0.Chain State Machine.AnimStateTransitionNode_5.Transition
    src=Attack_02
    via=AnimStateTransitionNode_5
   }
   {
    dst=Conduit
    rule=AnimGraph.AnimGraphNode_StateMachine_0.Chain State Machine.AnimStateTransitionNode_10.Transition
    src=Attack_02
    via=AnimStateTransitionNode_10
   }
   {
    dst=Conduit
    rule=AnimGraph.AnimGraphNode_StateMachine_0.Chain State Machine.AnimStateTransitionNode_6.Transition
    src=Attack_02
    via=AnimStateTransitionNode_6
   }
   {
    dst=Conduit
    rule=AnimGraph.AnimGraphNode_StateMachine_0.Chain State Machine.AnimStateTransitionNode_14.Transition
    src=Attack_02_Upper
    via=AnimStateTransitionNode_14
   }
   {
    dst=Conduit
    rule=AnimGraph.AnimGraphNode_StateMachine_0.Chain State Machine.AnimStateTransitionNode_16.Transition
    src=Attack_02_Upper
    via=AnimStateTransitionNode_16
   }
   {
    dst=Attack_01
    rule=AnimGraph.AnimGraphNode_StateMachine_0.Chain State Machine.AnimStateTransitionNode_8.Transition
    src=Attack_03
    via=AnimStateTransitionNode_8
   }
   {
    dst=Conduit
    rule=AnimGraph.AnimGraphNode_StateMachine_0.Chain State Machine.AnimStateTransitionNode_7.Transition
    src=Attack_03
    via=AnimStateTransitionNode_7
   }
   {
    dst=Conduit
    rule=AnimGraph.AnimGraphNode_StateMachine_0.Chain State Machine.AnimStateTransitionNode_9.Transition
    src=Attack_03
    via=AnimStateTransitionNode_9
   }
   {
    dst=Attack_01
    rule=AnimGraph.AnimGraphNode_StateMachine_0.Chain State Machine.AnimStateTransitionNode_2.Transition
    src=Conduit
    via=AnimStateTransitionNode_2
   }
   {
    dst=Attack_01_Upper
    rule=AnimGraph.AnimGraphNode_StateMachine_0.Chain State Machine.AnimStateTransitionNode_12.Transition
    src=Conduit
    via=AnimStateTransitionNode_12
   }
   {
    dst=Idle
    rule=AnimGraph.AnimGraphNode_StateMachine_0.Chain State Machine.AnimStateTransitionNode_0.Transition
    src=Conduit
    via=AnimStateTransitionNode_0
   }
   {
    dst=Conduit
    rule=AnimGraph.AnimGraphNode_StateMachine_0.Chain State Machine.AnimStateTransitionNode_1.Transition
    src=Idle
    via=AnimStateTransitionNode_1
   }
  ]
 }
 Conduit {path=AnimGraph.AnimGraphNode_StateMachine_0.Chain State Machine.AnimStateConduitNode_0.Conduit, nodes [Result(bCanEnterTransition=true)]}
 Idle {
  path=AnimGraph.AnimGraphNode_StateMachine_0.Chain State Machine.AnimStateNode_0.Idle
  nodes [Output Animation Pose(Result=Sequence Player 'Hades_Weapon_NonTargeted'() {bLoopAnimation=false, sequence=&Hades_Weapon_NonTargeted})]
 }
 Attack_01 {
  path=AnimGraph.AnimGraphNode_StateMachine_0.Chain State Machine.AnimStateNode_1.Attack_01
  nodes [
   AnimGraphNode_StateMachine_0 = Attack_01() {is=title AnimGraphNode_StateMachine, bAllowConduitEntryStates=true}
   Output Animation Pose(Result=@AnimGraphNode_StateMachine_0)
  ]
 }
 Attack_01 {
  entry=Conduit
  path=AnimGraph.AnimGraphNode_StateMachine_0.Chain State Machine.AnimStateNode_1.Attack_01.AnimGraphNode_StateMachine_0.Attack_01
  ids=id
  nodes [
   AnimStateConduitNode_0 = Conduit()
   AnimStateNode_0 = 1500() {is=title AnimStateNode}
   AnimStateNode_1 = 700() {is=title AnimStateNode}
   AnimStateNode_2 = 1200() {is=title AnimStateNode}
   AnimStateTransitionNode_0 = AnimStateTransitionNode() {priorityOrder=100}
   AnimStateTransitionNode_1 = AnimStateTransitionNode()
   AnimStateTransitionNode_2 = AnimStateTransitionNode() {priorityOrder=2}
  ]
  links [
   {
    dst=1200
    rule=AnimGraph.AnimGraphNode_StateMachine_0.Chain State Machine.AnimStateNode_1.Attack_01.AnimGraphNode_StateMachine_0.Attack_01.AnimStateTransitionNode_2.Transition
    src=Conduit
    via=AnimStateTransitionNode_2
   }
   {
    dst=1500
    rule=AnimGraph.AnimGraphNode_StateMachine_0.Chain State Machine.AnimStateNode_1.Attack_01.AnimGraphNode_StateMachine_0.Attack_01.AnimStateTransitionNode_0.Transition
    src=Conduit
    via=AnimStateTransitionNode_0
   }
   {
    dst=700
    rule=AnimGraph.AnimGraphNode_StateMachine_0.Chain State Machine.AnimStateNode_1.Attack_01.AnimGraphNode_StateMachine_0.Attack_01.AnimStateTransitionNode_1.Transition
    src=Conduit
    via=AnimStateTransitionNode_1
   }
  ]
 }
 Conduit {
  path=AnimGraph.AnimGraphNode_StateMachine_0.Chain State Machine.AnimStateNode_1.Attack_01.AnimGraphNode_StateMachine_0.Attack_01.AnimStateConduitNode_0.Conduit
  nodes [Result(bCanEnterTransition=true)]
 }
 1500 {
  path=AnimGraph.AnimGraphNode_StateMachine_0.Chain State Machine.AnimStateNode_1.Attack_01.AnimGraphNode_StateMachine_0.Attack_01.AnimStateNode_0.1500
  nodes [Output Animation Pose(Result=Two Way Blend(A=Use cached pose '1200'(), B=Use cached pose '1500'(), Alpha=Get AttackRange1500() -> (AttackRange1500:float)) {blendNode {alphaBoolBlend {blendOption=Linear}}})]
 }
 700 {
  path=AnimGraph.AnimGraphNode_StateMachine_0.Chain State Machine.AnimStateNode_1.Attack_01.AnimGraphNode_StateMachine_0.Attack_01.AnimStateNode_1.700
  nodes [Output Animation Pose(Result=Two Way Blend(A=Use cached pose '200'(), B=Use cached pose '700'(), Alpha=Get AttackRange700() -> (AttackRange700:float)) {blendNode {alphaBoolBlend {blendOption=Linear}}})]
 }
 1200 {
  path=AnimGraph.AnimGraphNode_StateMachine_0.Chain State Machine.AnimStateNode_1.Attack_01.AnimGraphNode_StateMachine_0.Attack_01.AnimStateNode_2.1200
  nodes [Output Animation Pose(Result=Two Way Blend(A=Use cached pose '700'(), B=Use cached pose '1200'(), Alpha=Get AttackRange1200() -> (AttackRange1200:float)) {blendNode {alphaBoolBlend {blendOption=Linear}}})]
 }
 Transition {
  path=AnimGraph.AnimGraphNode_StateMachine_0.Chain State Machine.AnimStateNode_1.Attack_01.AnimGraphNode_StateMachine_0.Attack_01.AnimStateTransitionNode_0.Transition
  nodes [Result(bCanEnterTransition=true)]
 }
 Transition {
  path=AnimGraph.AnimGraphNode_StateMachine_0.Chain State Machine.AnimStateNode_1.Attack_01.AnimGraphNode_StateMachine_0.Attack_01.AnimStateTransitionNode_1.Transition
  nodes [Result(bCanEnterTransition=Get bIs700Distance() -> (bIs700Distance:bool))]
 }
 Transition {
  path=AnimGraph.AnimGraphNode_StateMachine_0.Chain State Machine.AnimStateNode_1.Attack_01.AnimGraphNode_StateMachine_0.Attack_01.AnimStateTransitionNode_2.Transition
  nodes [Result(bCanEnterTransition=Get bIs1200Distance() -> (bIs1200Distance:bool))]
 }
 Attack_02 {
  path=AnimGraph.AnimGraphNode_StateMachine_0.Chain State Machine.AnimStateNode_2.Attack_02
  nodes [
   AnimGraphNode_StateMachine_0 = Attack_02() {is=title AnimGraphNode_StateMachine, bAllowConduitEntryStates=true}
   Output Animation Pose(Result=@AnimGraphNode_StateMachine_0)
  ]
 }
 Attack_02 {
  entry=Conduit
  path=AnimGraph.AnimGraphNode_StateMachine_0.Chain State Machine.AnimStateNode_2.Attack_02.AnimGraphNode_StateMachine_0.Attack_02
  ids=id
  nodes [
   AnimStateConduitNode_0 = Conduit()
   AnimStateNode_0 = 1500()
   AnimStateNode_1 = 700()
   AnimStateNode_2 = 1200()
   AnimStateTransitionNode_0 = AnimStateTransitionNode() {priorityOrder=100}
   AnimStateTransitionNode_1 = AnimStateTransitionNode()
   AnimStateTransitionNode_2 = AnimStateTransitionNode() {priorityOrder=2}
  ]
  links [
   {
    dst=1200
    rule=AnimGraph.AnimGraphNode_StateMachine_0.Chain State Machine.AnimStateNode_2.Attack_02.AnimGraphNode_StateMachine_0.Attack_02.AnimStateTransitionNode_2.Transition
    src=Conduit
    via=AnimStateTransitionNode_2
   }
   {
    dst=1500
    rule=AnimGraph.AnimGraphNode_StateMachine_0.Chain State Machine.AnimStateNode_2.Attack_02.AnimGraphNode_StateMachine_0.Attack_02.AnimStateTransitionNode_0.Transition
    src=Conduit
    via=AnimStateTransitionNode_0
   }
   {
    dst=700
    rule=AnimGraph.AnimGraphNode_StateMachine_0.Chain State Machine.AnimStateNode_2.Attack_02.AnimGraphNode_StateMachine_0.Attack_02.AnimStateTransitionNode_1.Transition
    src=Conduit
    via=AnimStateTransitionNode_1
   }
  ]
 }
 Conduit {
  path=AnimGraph.AnimGraphNode_StateMachine_0.Chain State Machine.AnimStateNode_2.Attack_02.AnimGraphNode_StateMachine_0.Attack_02.AnimStateConduitNode_0.Conduit
  nodes [Result(bCanEnterTransition=true)]
 }
 1500 {
  path=AnimGraph.AnimGraphNode_StateMachine_0.Chain State Machine.AnimStateNode_2.Attack_02.AnimGraphNode_StateMachine_0.Attack_02.AnimStateNode_0.1500
  nodes [Output Animation Pose(Result=Two Way Blend(A=Use cached pose '1200'(), B=Use cached pose '1500'(), Alpha=Get AttackRange1500()) {blendNode {alphaBoolBlend {blendOption=Linear}}})]
 }
 700 {
  path=AnimGraph.AnimGraphNode_StateMachine_0.Chain State Machine.AnimStateNode_2.Attack_02.AnimGraphNode_StateMachine_0.Attack_02.AnimStateNode_1.700
  nodes [Output Animation Pose(Result=Two Way Blend(A=Use cached pose '200'(), B=Use cached pose '700'(), Alpha=Get AttackRange700()) {blendNode {alphaBoolBlend {blendOption=Linear}}})]
 }
 1200 {
  path=AnimGraph.AnimGraphNode_StateMachine_0.Chain State Machine.AnimStateNode_2.Attack_02.AnimGraphNode_StateMachine_0.Attack_02.AnimStateNode_2.1200
  nodes [Output Animation Pose(Result=Two Way Blend(A=Use cached pose '700'(), B=Use cached pose '1200'(), Alpha=Get AttackRange1200()) {blendNode {alphaBoolBlend {blendOption=Linear}}})]
 }
 Transition {
  path=AnimGraph.AnimGraphNode_StateMachine_0.Chain State Machine.AnimStateNode_2.Attack_02.AnimGraphNode_StateMachine_0.Attack_02.AnimStateTransitionNode_0.Transition
  nodes [Result(bCanEnterTransition=true)]
 }
 Transition {
  path=AnimGraph.AnimGraphNode_StateMachine_0.Chain State Machine.AnimStateNode_2.Attack_02.AnimGraphNode_StateMachine_0.Attack_02.AnimStateTransitionNode_1.Transition
  nodes [Result(bCanEnterTransition=Get bIs700Distance())]
 }
 Transition {
  path=AnimGraph.AnimGraphNode_StateMachine_0.Chain State Machine.AnimStateNode_2.Attack_02.AnimGraphNode_StateMachine_0.Attack_02.AnimStateTransitionNode_2.Transition
  nodes [Result(bCanEnterTransition=Get bIs1200Distance())]
 }
 Attack_03 {
  path=AnimGraph.AnimGraphNode_StateMachine_0.Chain State Machine.AnimStateNode_3.Attack_03
  nodes [
   AnimGraphNode_StateMachine_0 = Attack_02() {bAllowConduitEntryStates=true}
   Output Animation Pose(Result=@AnimGraphNode_StateMachine_0)
  ]
 }
 Attack_02 {
  entry=Conduit
  path=AnimGraph.AnimGraphNode_StateMachine_0.Chain State Machine.AnimStateNode_3.Attack_03.AnimGraphNode_StateMachine_0.Attack_02
  ids=id
  nodes [
   AnimStateConduitNode_0 = Conduit()
   AnimStateNode_0 = 1500()
   AnimStateNode_1 = 700()
   AnimStateNode_2 = 1200()
   AnimStateTransitionNode_0 = AnimStateTransitionNode() {priorityOrder=100}
   AnimStateTransitionNode_1 = AnimStateTransitionNode()
   AnimStateTransitionNode_2 = AnimStateTransitionNode() {priorityOrder=2}
  ]
  links [
   {
    dst=1200
    rule=AnimGraph.AnimGraphNode_StateMachine_0.Chain State Machine.AnimStateNode_3.Attack_03.AnimGraphNode_StateMachine_0.Attack_02.AnimStateTransitionNode_2.Transition
    src=Conduit
    via=AnimStateTransitionNode_2
   }
   {
    dst=1500
    rule=AnimGraph.AnimGraphNode_StateMachine_0.Chain State Machine.AnimStateNode_3.Attack_03.AnimGraphNode_StateMachine_0.Attack_02.AnimStateTransitionNode_0.Transition
    src=Conduit
    via=AnimStateTransitionNode_0
   }
   {
    dst=700
    rule=AnimGraph.AnimGraphNode_StateMachine_0.Chain State Machine.AnimStateNode_3.Attack_03.AnimGraphNode_StateMachine_0.Attack_02.AnimStateTransitionNode_1.Transition
    src=Conduit
    via=AnimStateTransitionNode_1
   }
  ]
 }
 Conduit {
  path=AnimGraph.AnimGraphNode_StateMachine_0.Chain State Machine.AnimStateNode_3.Attack_03.AnimGraphNode_StateMachine_0.Attack_02.AnimStateConduitNode_0.Conduit
  nodes [Result(bCanEnterTransition=true)]
 }
 1500 {
  path=AnimGraph.AnimGraphNode_StateMachine_0.Chain State Machine.AnimStateNode_3.Attack_03.AnimGraphNode_StateMachine_0.Attack_02.AnimStateNode_0.1500
  nodes [Output Animation Pose(Result=Two Way Blend(A=Use cached pose '1200'(), B=Use cached pose '1500'(), Alpha=Get AttackRange1500()) {blendNode {alphaBoolBlend {blendOption=Linear}}})]
 }
 700 {
  path=AnimGraph.AnimGraphNode_StateMachine_0.Chain State Machine.AnimStateNode_3.Attack_03.AnimGraphNode_StateMachine_0.Attack_02.AnimStateNode_1.700
  nodes [Output Animation Pose(Result=Two Way Blend(A=Use cached pose '200'(), B=Use cached pose '700'(), Alpha=Get AttackRange700()) {blendNode {alphaBoolBlend {blendOption=Linear}}})]
 }
 1200 {
  path=AnimGraph.AnimGraphNode_StateMachine_0.Chain State Machine.AnimStateNode_3.Attack_03.AnimGraphNode_StateMachine_0.Attack_02.AnimStateNode_2.1200
  nodes [Output Animation Pose(Result=Two Way Blend(A=Use cached pose '700'(), B=Use cached pose '1200'(), Alpha=Get AttackRange1200()) {blendNode {alphaBoolBlend {blendOption=Linear}}})]
 }
 Transition {
  path=AnimGraph.AnimGraphNode_StateMachine_0.Chain State Machine.AnimStateNode_3.Attack_03.AnimGraphNode_StateMachine_0.Attack_02.AnimStateTransitionNode_0.Transition
  nodes [Result(bCanEnterTransition=true)]
 }
 Transition {
  path=AnimGraph.AnimGraphNode_StateMachine_0.Chain State Machine.AnimStateNode_3.Attack_03.AnimGraphNode_StateMachine_0.Attack_02.AnimStateTransitionNode_1.Transition
  nodes [Result(bCanEnterTransition=Get bIs700Distance())]
 }
 Transition {
  path=AnimGraph.AnimGraphNode_StateMachine_0.Chain State Machine.AnimStateNode_3.Attack_03.AnimGraphNode_StateMachine_0.Attack_02.AnimStateTransitionNode_2.Transition
  nodes [Result(bCanEnterTransition=Get bIs1200Distance())]
 }
 Attack_01_Upper {
  path=AnimGraph.AnimGraphNode_StateMachine_0.Chain State Machine.AnimStateNode_4.Attack_01_Upper
  nodes [
   AnimGraphNode_StateMachine_0 = Attack_01() {bAllowConduitEntryStates=true}
   Output Animation Pose(Result=@AnimGraphNode_StateMachine_0)
  ]
 }
 Attack_01 {
  entry=Conduit
  path=AnimGraph.AnimGraphNode_StateMachine_0.Chain State Machine.AnimStateNode_4.Attack_01_Upper.AnimGraphNode_StateMachine_0.Attack_01
  ids=id
  nodes [
   AnimStateConduitNode_0 = Conduit()
   AnimStateNode_0 = 1500()
   AnimStateNode_1 = 700()
   AnimStateNode_2 = 1200()
   AnimStateTransitionNode_0 = AnimStateTransitionNode() {crossfadeDuration=0, priorityOrder=100}
   AnimStateTransitionNode_1 = AnimStateTransitionNode() {crossfadeDuration=0}
   AnimStateTransitionNode_2 = AnimStateTransitionNode() {crossfadeDuration=0, priorityOrder=2}
  ]
  links [
   {
    dst=1200
    rule=AnimGraph.AnimGraphNode_StateMachine_0.Chain State Machine.AnimStateNode_4.Attack_01_Upper.AnimGraphNode_StateMachine_0.Attack_01.AnimStateTransitionNode_2.Transition
    src=Conduit
    via=AnimStateTransitionNode_2
   }
   {
    dst=1500
    rule=AnimGraph.AnimGraphNode_StateMachine_0.Chain State Machine.AnimStateNode_4.Attack_01_Upper.AnimGraphNode_StateMachine_0.Attack_01.AnimStateTransitionNode_0.Transition
    src=Conduit
    via=AnimStateTransitionNode_0
   }
   {
    dst=700
    rule=AnimGraph.AnimGraphNode_StateMachine_0.Chain State Machine.AnimStateNode_4.Attack_01_Upper.AnimGraphNode_StateMachine_0.Attack_01.AnimStateTransitionNode_1.Transition
    src=Conduit
    via=AnimStateTransitionNode_1
   }
  ]
 }
 Conduit {
  path=AnimGraph.AnimGraphNode_StateMachine_0.Chain State Machine.AnimStateNode_4.Attack_01_Upper.AnimGraphNode_StateMachine_0.Attack_01.AnimStateConduitNode_0.Conduit
  nodes [Result(bCanEnterTransition=true)]
 }
 1500 {
  path=AnimGraph.AnimGraphNode_StateMachine_0.Chain State Machine.AnimStateNode_4.Attack_01_Upper.AnimGraphNode_StateMachine_0.Attack_01.AnimStateNode_0.1500
  nodes [Output Animation Pose(Result=Two Way Blend(A=Use cached pose '1200'(), B=Use cached pose '1500'(), Alpha=Get AttackRange1500()) {blendNode {alphaBoolBlend {blendOption=Linear}}})]
 }
 700 {
  path=AnimGraph.AnimGraphNode_StateMachine_0.Chain State Machine.AnimStateNode_4.Attack_01_Upper.AnimGraphNode_StateMachine_0.Attack_01.AnimStateNode_1.700
  nodes [Output Animation Pose(Result=Two Way Blend(A=Use cached pose '200'(), B=Use cached pose '700'(), Alpha=Get AttackRange700()) {blendNode {alphaBoolBlend {blendOption=Linear}}})]
 }
 1200 {
  path=AnimGraph.AnimGraphNode_StateMachine_0.Chain State Machine.AnimStateNode_4.Attack_01_Upper.AnimGraphNode_StateMachine_0.Attack_01.AnimStateNode_2.1200
  nodes [Output Animation Pose(Result=Two Way Blend(A=Use cached pose '700'(), B=Use cached pose '1200'(), Alpha=Get AttackRange1200()) {blendNode {alphaBoolBlend {blendOption=Linear}}})]
 }
 Transition {
  path=AnimGraph.AnimGraphNode_StateMachine_0.Chain State Machine.AnimStateNode_4.Attack_01_Upper.AnimGraphNode_StateMachine_0.Attack_01.AnimStateTransitionNode_0.Transition
  nodes [Result(bCanEnterTransition=true)]
 }
 Transition {
  path=AnimGraph.AnimGraphNode_StateMachine_0.Chain State Machine.AnimStateNode_4.Attack_01_Upper.AnimGraphNode_StateMachine_0.Attack_01.AnimStateTransitionNode_1.Transition
  nodes [Result(bCanEnterTransition=Get bIs700Distance())]
 }
 Transition {
  path=AnimGraph.AnimGraphNode_StateMachine_0.Chain State Machine.AnimStateNode_4.Attack_01_Upper.AnimGraphNode_StateMachine_0.Attack_01.AnimStateTransitionNode_2.Transition
  nodes [Result(bCanEnterTransition=Get bIs1200Distance())]
 }
 Attack_02_Upper {
  path=AnimGraph.AnimGraphNode_StateMachine_0.Chain State Machine.AnimStateNode_5.Attack_02_Upper
  nodes [
   AnimGraphNode_StateMachine_0 = Attack_02() {bAllowConduitEntryStates=true}
   Output Animation Pose(Result=@AnimGraphNode_StateMachine_0)
  ]
 }
 Attack_02 {
  entry=Conduit
  path=AnimGraph.AnimGraphNode_StateMachine_0.Chain State Machine.AnimStateNode_5.Attack_02_Upper.AnimGraphNode_StateMachine_0.Attack_02
  ids=id
  nodes [
   AnimStateConduitNode_0 = Conduit()
   AnimStateNode_0 = 1500()
   AnimStateNode_1 = 700()
   AnimStateNode_2 = 1200()
   AnimStateTransitionNode_0 = AnimStateTransitionNode() {priorityOrder=100}
   AnimStateTransitionNode_1 = AnimStateTransitionNode()
   AnimStateTransitionNode_2 = AnimStateTransitionNode() {priorityOrder=2}
  ]
  links [
   {
    dst=1200
    rule=AnimGraph.AnimGraphNode_StateMachine_0.Chain State Machine.AnimStateNode_5.Attack_02_Upper.AnimGraphNode_StateMachine_0.Attack_02.AnimStateTransitionNode_2.Transition
    src=Conduit
    via=AnimStateTransitionNode_2
   }
   {
    dst=1500
    rule=AnimGraph.AnimGraphNode_StateMachine_0.Chain State Machine.AnimStateNode_5.Attack_02_Upper.AnimGraphNode_StateMachine_0.Attack_02.AnimStateTransitionNode_0.Transition
    src=Conduit
    via=AnimStateTransitionNode_0
   }
   {
    dst=700
    rule=AnimGraph.AnimGraphNode_StateMachine_0.Chain State Machine.AnimStateNode_5.Attack_02_Upper.AnimGraphNode_StateMachine_0.Attack_02.AnimStateTransitionNode_1.Transition
    src=Conduit
    via=AnimStateTransitionNode_1
   }
  ]
 }
 Conduit {
  path=AnimGraph.AnimGraphNode_StateMachine_0.Chain State Machine.AnimStateNode_5.Attack_02_Upper.AnimGraphNode_StateMachine_0.Attack_02.AnimStateConduitNode_0.Conduit
  nodes [Result(bCanEnterTransition=true)]
 }
 1500 {
  path=AnimGraph.AnimGraphNode_StateMachine_0.Chain State Machine.AnimStateNode_5.Attack_02_Upper.AnimGraphNode_StateMachine_0.Attack_02.AnimStateNode_0.1500
  nodes [Output Animation Pose(Result=Two Way Blend(A=Use cached pose '1200'(), B=Use cached pose '1500'(), Alpha=Get AttackRange1500()) {blendNode {alphaBoolBlend {blendOption=Linear}}})]
 }
 700 {
  path=AnimGraph.AnimGraphNode_StateMachine_0.Chain State Machine.AnimStateNode_5.Attack_02_Upper.AnimGraphNode_StateMachine_0.Attack_02.AnimStateNode_1.700
  nodes [Output Animation Pose(Result=Two Way Blend(A=Use cached pose '200'(), B=Use cached pose '700'(), Alpha=Get AttackRange700()) {blendNode {alphaBoolBlend {blendOption=Linear}}})]
 }
 1200 {
  path=AnimGraph.AnimGraphNode_StateMachine_0.Chain State Machine.AnimStateNode_5.Attack_02_Upper.AnimGraphNode_StateMachine_0.Attack_02.AnimStateNode_2.1200
  nodes [Output Animation Pose(Result=Two Way Blend(A=Use cached pose '700'(), B=Use cached pose '1200'(), Alpha=Get AttackRange1200()) {blendNode {alphaBoolBlend {blendOption=Linear}}})]
 }
 Transition {
  path=AnimGraph.AnimGraphNode_StateMachine_0.Chain State Machine.AnimStateNode_5.Attack_02_Upper.AnimGraphNode_StateMachine_0.Attack_02.AnimStateTransitionNode_0.Transition
  nodes [Result(bCanEnterTransition=true)]
 }
 Transition {
  path=AnimGraph.AnimGraphNode_StateMachine_0.Chain State Machine.AnimStateNode_5.Attack_02_Upper.AnimGraphNode_StateMachine_0.Attack_02.AnimStateTransitionNode_1.Transition
  nodes [Result(bCanEnterTransition=Get bIs700Distance())]
 }
 Transition {
  path=AnimGraph.AnimGraphNode_StateMachine_0.Chain State Machine.AnimStateNode_5.Attack_02_Upper.AnimGraphNode_StateMachine_0.Attack_02.AnimStateTransitionNode_2.Transition
  nodes [Result(bCanEnterTransition=Get bIs1200Distance())]
 }
 Transition {
  path=AnimGraph.AnimGraphNode_StateMachine_0.Chain State Machine.AnimStateTransitionNode_0.Transition
  nodes [Result(bCanEnterTransition=true)]
 }
 Transition {
  path=AnimGraph.AnimGraphNode_StateMachine_0.Chain State Machine.AnimStateTransitionNode_1.Transition
  nodes [Result(bCanEnterTransition=true)]
 }
 Transition {path=AnimGraph.AnimGraphNode_StateMachine_0.Chain State Machine.AnimStateTransitionNode_10.Transition, nodes [Result()]}
 Transition {path=AnimGraph.AnimGraphNode_StateMachine_0.Chain State Machine.AnimStateTransitionNode_11.Transition, nodes [Result()]}
 Transition {
  path=AnimGraph.AnimGraphNode_StateMachine_0.Chain State Machine.AnimStateTransitionNode_12.Transition
  nodes [Result(bCanEnterTransition=Get bIsUpperBodyAttack_01() -> (bIsUpperBodyAttack_01:bool))]
 }
 Transition {
  path=AnimGraph.AnimGraphNode_StateMachine_0.Chain State Machine.AnimStateTransitionNode_13.Transition
  nodes [Result(bCanEnterTransition=Get bIsUpperBodyAttack_02() -> (bIsUpperBodyAttack_02:bool))]
 }
 Transition {path=AnimGraph.AnimGraphNode_StateMachine_0.Chain State Machine.AnimStateTransitionNode_14.Transition, nodes [Result()]}
 Transition {path=AnimGraph.AnimGraphNode_StateMachine_0.Chain State Machine.AnimStateTransitionNode_15.Transition, nodes [Result()]}
 Transition {
  path=AnimGraph.AnimGraphNode_StateMachine_0.Chain State Machine.AnimStateTransitionNode_16.Transition
  nodes [Result(bCanEnterTransition=Get bIsMantisDone() -> (bIsMantisDone:bool))]
 }
 Transition {
  path=AnimGraph.AnimGraphNode_StateMachine_0.Chain State Machine.AnimStateTransitionNode_17.Transition
  nodes [Result(bCanEnterTransition=Get bIsMantisDone())]
 }
 Transition {
  path=AnimGraph.AnimGraphNode_StateMachine_0.Chain State Machine.AnimStateTransitionNode_2.Transition
  nodes [Result(bCanEnterTransition=Get bIsAttack_01() -> (bIsAttack_01:bool))]
 }
 Transition {
  path=AnimGraph.AnimGraphNode_StateMachine_0.Chain State Machine.AnimStateTransitionNode_3.Transition
  nodes [Result(bCanEnterTransition=Get bIsMantisDone())]
 }
 Transition {
  path=AnimGraph.AnimGraphNode_StateMachine_0.Chain State Machine.AnimStateTransitionNode_4.Transition
  nodes [Result(bCanEnterTransition=Get bIsAttack_02() -> (bIsAttack_02:bool))]
 }
 Transition {
  path=AnimGraph.AnimGraphNode_StateMachine_0.Chain State Machine.AnimStateTransitionNode_5.Transition
  nodes [Result(bCanEnterTransition=Get bIsAttack_03() -> (bIsAttack_03:bool))]
 }
 Transition {
  path=AnimGraph.AnimGraphNode_StateMachine_0.Chain State Machine.AnimStateTransitionNode_6.Transition
  nodes [Result(bCanEnterTransition=Get bIsMantisDone())]
 }
 Transition {
  path=AnimGraph.AnimGraphNode_StateMachine_0.Chain State Machine.AnimStateTransitionNode_7.Transition
  nodes [Result(bCanEnterTransition=Get bIsMantisDone())]
 }
 Transition {
  path=AnimGraph.AnimGraphNode_StateMachine_0.Chain State Machine.AnimStateTransitionNode_8.Transition
  nodes [Result(bCanEnterTransition=Get bIsAttack_01())]
 }
 Transition {path=AnimGraph.AnimGraphNode_StateMachine_0.Chain State Machine.AnimStateTransitionNode_9.Transition, nodes [Result()]}
 EventGraph {nodes [
   n1 = UpdateAttackRange!() -> (OutputDelegate, then, NewRange:double=~) {is=title K2Node_CustomEvent}
   > Set AttackRange(AttackRange:float=@n1.NewRange) -> (Output_Get:float)
  ]}
]
variables []
